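# Physics asset for the COD Squad Hoodie backpack cosmetic, defining bodies and joints for its radio, antenna and two shell pieces.
# A large box covers the pack, with tiny master boxes for the radio and each shell joined to it by angularly free constraints.
# The radio and shells attach to their masters with tight swing and twist limits, and the spherical antenna body chains off the radio.
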PhysicsAsset /BRCosmetics/Accessories/FORT_Backpacks/M_MED_COD_Squad_Hoodie_Backpack/Meshes/M_MED_COD_Squad_Hoodie_Pack_Physics
bodies {
 dyn_antenna [{shapeType=Sphere, radius=3.1315, center=<1.559 0 -0.5971>}]
 dyn_radio [{shapeType=Box, extentX=3.2693, extentY=9.7878, extentZ=22.644, center=<2.6587 0.1878 3.6953>}]
 dyn_radio_mstr [{shapeType=Box, extentX=1.01, extentY=1.01, extentZ=1.01}]
 dyn_shell_a [{shapeType=Box, extentX=5.9472, extentY=9.1696, extentZ=6.664, center=<-0.9995 0.0283 -0.144>}]
 dyn_shell_a_mstr [{shapeType=Box, extentX=1.01, extentY=1.01, extentZ=1.01}]
 dyn_shell_b [{shapeType=Box, extentX=3.6527, extentY=9.4623, extentZ=6.5227, center=<0.3281 -0.1772 -0.47>}]
 dyn_shell_b_mstr [{shapeType=Box, extentX=1.01, extentY=1.01, extentZ=1.01}]
 pack [{shapeType=Box, extentX=20.4715, extentY=11.744, extentZ=36.9015, center=<-1.1236 -10.0421 1.4627>}]
}
constraints [
 {
  bone1Name=dyn_radio_mstr
  bone2Name=pack
  swing1LimitDegrees=45
  swing1Motion=Free
  swing2LimitDegrees=45
  swing2Motion=Free
  twistLimitDegrees=45
  twistMotion=Free
 }
 {
  bone1Name=dyn_shell_a_mstr
  bone2Name=pack
  swing1LimitDegrees=45
  swing1Motion=Free
  swing2LimitDegrees=45
  swing2Motion=Free
  twistLimitDegrees=45
  twistMotion=Free
 }
 {
  bone1Name=dyn_shell_b_mstr
  bone2Name=pack
  swing1LimitDegrees=45
  swing1Motion=Free
  swing2LimitDegrees=45
  swing2Motion=Free
  twistLimitDegrees=45
  twistMotion=Free
 }
 {bone1Name=dyn_radio, bone2Name=dyn_radio_mstr, swing1LimitDegrees=15, swing2LimitDegrees=15, twistLimitDegrees=5}
 {bone1Name=dyn_shell_a, bone2Name=dyn_shell_a_mstr, swing1LimitDegrees=15, swing2LimitDegrees=15, twistLimitDegrees=5}
 {bone1Name=dyn_shell_b, bone2Name=dyn_shell_b_mstr, swing1LimitDegrees=15, swing2LimitDegrees=15, twistLimitDegrees=5}
 {bone1Name=dyn_antenna, bone2Name=dyn_radio, swing1LimitDegrees=20, swing2LimitDegrees=20, twistLimitDegrees=5}
]
